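# Sound cue for the Creative Item Spawner toy, named for its item-spawn hologram sound.
# It plays a single wave player using the Card_Summary sound from the front-end store UI, at a much reduced volume under the Creative gadgets sound class.
# It references the Creative small-toy attenuation but overrides it with a short falloff and listener focus.
# A concurrency override caps simultaneous plays at a low count, stopping the quietest.
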SoundCue /Game/Sounds/Creative/Toys/ItemSpawner/ItemSpawner_ItemSpawnHolo_Cue
root=WavePlayer_13
nodes [{id=WavePlayer_13, properties {soundWaveAssetPtr=/Game/Sounds/Fort_UI_Sounds/FrontEnd/Store/wav/Card_Summary}}]
properties {
 attenuationSettings=/Game/Sounds/Creative/Attenuations/Fort_Creative_SmallToy_Att
 bOverrideAttenuation=true
 bOverrideConcurrency=true
 maxDistance=500
 soundClassObject=&Fort_Gadgets_Creative
 volumeMultiplier=0.1
 attenuationOverrides {bEnableListenerFocus=true, falloffDistance=400, nonFocusVolumeAttenuation=0.5, attenuationShapeExtents {x=100}}
 concurrencyOverrides {maxCount=2, resolutionRule=StopQuietest}
}
