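# Sound cue for the Creative Item Spawner toy, played when it spawns an item.
# A single wave player feeds the spawn sound directly to the output, with no randomization or modulation.
# It uses the Creative small-toy attenuation and the Creative gadgets sound class.
# Its own concurrency override caps simultaneous instances at a low count and blocks new plays rather than cutting off existing ones.
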SoundCue /Game/Sounds/Creative/Toys/ItemSpawner/ItemSpawner_ItemSpawn_Cue
root=WavePlayer_0
nodes [{id=WavePlayer_0, properties {soundWaveAssetPtr=&/ItemSpawner_ItemSpawn}}]
properties {
 attenuationSettings=/Game/Sounds/Creative/Attenuations/Fort_Creative_SmallToy_Att
 bOverrideConcurrency=true
 maxDistance=3000
 soundClassObject=&Fort_Gadgets_Creative
 concurrencyOverrides {maxCount=2, resolutionRule=PreventNew, voiceStealReleaseTime=0.5}
}
