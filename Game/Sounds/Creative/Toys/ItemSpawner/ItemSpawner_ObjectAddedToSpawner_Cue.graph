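# Sound cue for the Creative item spawner toy, named for the moment an object is added to the spawner.
# A single wave player feeds the ItemSpawner_ObjectAddedToSpawner sound straight to the output, with no randomization or mixing.
# It plays quietly through a reduced volume multiplier, uses the Creative small-toy attenuation for spatial falloff, and routes to the Creative gadgets sound class.
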SoundCue /Game/Sounds/Creative/Toys/ItemSpawner/ItemSpawner_ObjectAddedToSpawner_Cue
root=WavePlayer_0
nodes [{id=WavePlayer_0, properties {soundWaveAssetPtr=&/ItemSpawner_ObjectAddedToSpawner}}]
properties {
 attenuationSettings=/Game/Sounds/Creative/Attenuations/Fort_Creative_SmallToy_Att
 maxDistance=3000
 soundClassObject=&Fort_Gadgets_Creative
 volumeMultiplier=0.2
}
